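; Heist base manager Blueprint that runs one base's lockdown, matched by base tag.
; On authority it reserves AI budget, caches defenses from the actor registry (tagged by an editor sphere-trace function) and, on a lockdown message, notifies them by interface, broadcasts and starts optional expiry.
; Area overlaps give players loose base and status tags and feed analytics sent at EndGame.
; RepNotifies drive client lockdown visuals through a global material parameter collection.
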(event EventBeginPlay
  ;[0.04468 1 0] BeginPlay
  ; Cache the base defenses based on the Actor tag
  ; and listen for the Lockdown message
  (|CallParentFunction)
  (Networking|SwitchHasAuthority
    (:Authority
      (bind (_gamestate _playlist) (Class|FortAsyncActionCurrentPlaylistReady|PlaylistDataReadyAsync)
        (:OnReady
          (Utilities|FlowControl|Sequence
            (:then_0
              ;;[0.1418 0.1369 1] Sequence[0]: If we want swappable data layers add the component and pass the data layers associated to this base to the component
              (if (ScalableFloat|GetValueasBool (Variables|Hotfix|GetHotfix_HeistBaseDataLayersEnabled))
                (bind _returnvalue (Game|AddComponentbyClass (Variables|DataLayer|GetHeistBaseDataLayerComponentClass)))
                (bind _asb_heist_base_data_layer_component (Utilities|Casting|CastToB_HeistBaseDataLayerComponent _returnvalue))
                (bind _output_get (Variables|DataLayer|SetHeistBaseDataLayerComponent _asb_heist_base_data_layer_component))
                (Class|BHeistBaseDataLayerComponent|InitializeDataLayers _output_get (Variables|DataLayer|GetDefenseDataLayers))))
            (:then_1
              ;;[0.1418 0.1369 1] Sequence [1] Listen for the lockdown for the base and begin lockdown
              (bind (_asynctaskproxy _payload) (Messaging|RegisterGameplayMessageReceiver (Variables|Default|GetLockdownRequestedMessageChannel) "/HeistBasesGameplay/Items/Struct_HeistBases_LockdownInitiated_Payload.Struct_HeistBases_LockdownInitiated_Payload")
                (:then
                  (Variables|Default|SetLockdown%20Request%20GameplayMessageTask _asynctaskproxy))
                (:OnMessageReceived
                  (if (== (Variables|Default|GetHeistBaseId%20Tag%20Const) (Utilities|Struct|BreakStructHeistBasesLockdownInitiatedPayload _payload))
                    (CallFunction|OnRequestLockdownBase _payload)
                    ;;;[1 0 0.9102] LOG: Valid Gameplay message received
                    (Fortnite|FortShippingLog self (Utilities|String|Append "HEIST BASE: Lockdown base message (" (GameplayTags|GetDebugStringfromGameplayTag (Utilities|Struct|BreakStructHeistBasesLockdownInitiatedPayload _payload)) ") is valid and received."))
                    (else
                      ;;[1 0 0.9102] LOG: Valid invalid message recieved - this is expected for the other two Heist Bases
                      (Fortnite|FortShippingLog self (Utilities|String|Append "HEIST BASE: Lockdown base message (" (GameplayTags|GetDebugStringfromGameplayTag (Utilities|Struct|BreakStructHeistBasesLockdownInitiatedPayload _payload)) ") does not match base (" "" "). This is expected for the other base instances")))))))
            (:then_2
              ;;[0.1418 0.1369 1] Sequence[2]: Reserve AI Budget for this base
              (if (ScalableFloat|GetValueasBool (Variables|Hotfix|GetHotfix_AISpawningEnabled))
                (Protected|AI|ReserveAIDensityReservoir false)))
            (:then_3
              ;;[0.1418 0.1369 1] Sequence[3] Get and cache the Defense objects registered for this base (have the same base tag)
              (CallFunction|GetandRegisterDefenses))
            (:then_4
              ;;[0.2435 0.1798 0.4062] Sequence [4]
              ;; Analytics: Bind to the Game Phase changed to get the end of match to send the event
              (Utilities|IsValid _gamestate
                (:Is%20Valid
                  (bind _returnvalue_2 (GameplayEventRouter|GetEventRouter "/Script/FortniteGame.FortGameState" self))
                  (bind (_asynctaskproxy_2 _payload_2) (GameplayEventRouter|StartListeningToStatefulEvent _returnvalue_2 "/Script/FortniteGame.FortGamePhaseUpdatedEvent")
                    (:then
                      (Variables|Default|SetGamePhaseUpdatedAsyncTask _asynctaskproxy_2))
                    (:OnEventReceived
                      _)
                    (:OnSavedEventStateExists
                      _))
                  (Analytics|HandleGamePhaseChanged (Utilities|Struct|BreakFortGamePhaseUpdatedEvent _payload_2)))))
            (:then_5
              ;;[0.1407 0.1494 1] Sequence [5]
              ;; Cache the playlists max team size for use with AI scaling
              (Utilities|IsValid _playlist
                (:Is%20Valid
                  (Variables|AI|SetCurrentPlaylistMaxTeamSize (Variables|/Script/FortniteGame.FortPlaylist|GetMaxTeamSize _playlist)))))))))
    (:Remote
      ;;[1 0 0.8953] Disable collision on clients.
      (Collision|SetCollisionEnabled (Variables|Default|GetHeistBaseAreaCollision) "NoCollision"))))

(event Custom|OnRequestLockdownBase
  (params (Payload :type Struct%20Heist%20Bases%20Lockdown%20Initiated%20Payload%20Structure))
  (bind (_basemanagertag_16_f65d8231436245dc45908a82e6cac87c _triggeringpawn_9_567a525a4447d3f140bec19d19506604 _triggeringactor_11_c6d5ffac409c2b65b7d64d89bf7cf0c5) (Utilities|Struct|BreakStructHeistBasesLockdownInitiatedPayload Payload))
  (bind _lockdownbeforeexpiredduration (Variables|Lockdown|GetLockdownBeforeExpiredDuration))
  (bind _returnvalue (ScalableFloat|GetValueatLevel _lockdownbeforeexpiredduration))
  (bind _cachedbasedefenses (Variables|BaseDefense|GetCachedBaseDefenses))
  ;[0.2448 0.1341 0.03254] OnLockdownBase
  ; Initate the lockdown behavior for any base defenses
  ; (lockdown gates, display cases, guard elevators, turrets, etc.)
  (Networking|SwitchHasAuthority
    (:Authority
      (if (not (Variables|Lockdown|GetbInLockdown))
        (Utilities|FlowControl|Sequence
          (:then_0
            ;;[0.2448 0.1341 0.03254] [0] Initiate Lockdown for base and all registered defenses
            (Variables|Lockdown|SetbInLockdown true)
            (Networking|FlushNetDormancy)
            (CallFunction|SetStatus "(TagName=\"HeistBases.Status.Lockeddown\")")
            (if (> _returnvalue 0)
              (bind _returnvalue_2 (Utilities|Time|SetTimerbyEvent (EventDispatchers|CreateEvent :!event Custom|EndLockdownByExpiration) _returnvalue))
              (Variables|Lockdown|SetTH_LockdownExpirationTimer _returnvalue_2)
              (else
                _))
            (for _array_element _cachedbasedefenses
              (Utilities|IsValid _array_element
                (:Is%20Valid
                  (Class|BPIHeistBasesLockdown|OnLockdownInitiated(Message) _array_element))))
            ;;;[1 0 0.9339] LOG: When lockdown occurs report the number of defenses that have been cached for lockdown behavior
            (Fortnite|FortShippingLog self (Utilities|String|Append "HEIST LOCKDOWN: Base (" (GameplayTags|GetDebugStringfromGameplayTag (Variables|Default|GetHeistBaseId%20Tag%20Const)) ") has locked down [" (Utilities|String|ToString(Integer) (Utilities|Array|Length _cachedbasedefenses)) "] base defenses.")))
          (:then_1
            ;;[0.3407 0.06221 1] Broadcast that heist base lockdown has initiated - for Quests, analytics, and Lockdown visuals
            ;;;[0.2435 0.1798 0.4062] Analytics: store relevant player for evaluations
            (bind _asfort_player_pawn_athena (Utilities|Casting|CastToFortPlayerPawnAthena _triggeringpawn_9_567a525a4447d3f140bec19d19506604)
              (:then
                (Variables|Analytics|SetLockdownTriggeringPawnAthena _asfort_player_pawn_athena))
              (:CastFailed
                _))
            (Messaging|BroadcastMessage (GameInstanceSubsystems|GetGameplayMessageRouter) (Variables|Default|GetLockdownStartedMessageChannel) (Utilities|Struct|MakeGameplayEventData (Variables|Default|GetLockdownStartedMessageChannel) _triggeringactor_11_c6d5ffac409c2b65b7d64d89bf7cf0c5 self _triggeringpawn_9_567a525a4447d3f140bec19d19506604 0 (Class|AthenaFunctionLibrary|MakeSimpleFortEffectContext _triggeringactor_11_c6d5ffac409c2b65b7d64d89bf7cf0c5) (GameplayTags|MakeGameplayTagContainerfromTag _basemanagertag_16_f65d8231436245dc45908a82e6cac87c) (GameplayTags|MakeGameplayTagContainerfromTag (Variables|Default|GetHeistBaseId%20Tag%20Const))) false self)
            ;;[0.2435 0.1798 0.4062] LOCKDOWN ANALYTICS
            ;;;[0.2435 0.1798 0.4062] Analytic: Base Within the Storm On Lockdown triggered?
            (bind _returnvalue_3 (SafeZone|IsLocationinSafeZone (Transformation|GetActorLocation)))
            (Variables|Analytics|SetAnalytic_BaseInStormOnLockdown (not _returnvalue_3))
            ;;;[0.2435 0.1798 0.4062] Analytic: The time it took for the lockdown-initiating player to enter the zone and then claim the reward (initiate lockdown)
            (bind _lockdowntriggeringpawnathena (Variables|Analytics|GetLockdownTriggeringPawnAthena)
              (:then
                (bind _output_get (Variables|Analytics|SetAnalytic_LockdownTime (Utilities|Time|GetGameTimeinSeconds)))
                (Variables|Analytics|SetAnalytic_PlayerLockdownTime (- _output_get (Utilities|Map|Find (Variables|Analytics|GetAnalyticsMapPlayerToEntryTime) _lockdowntriggeringpawnathena))))
              (:else))))))))

(event Custom|EndLockdownByExpiration
  ;[0.8385 0.3198 0] EndLockdownByExpiration
  ; When the lockdown should expire,
  ; send the expiration event to the interface of all defense objects
  (Variables|Lockdown|SetbInLockdown false)
  (Variables|Lockdown|SetbLockdownExpired true)
  (Networking|FlushNetDormancy)
  (CallFunction|SetStatus "(TagName=\"HeistBases.Status.Unlocked\")")
  (for _array_element (Variables|BaseDefense|GetCachedBaseDefenses)
    (Utilities|IsValid _array_element
      (:Is%20Valid
        (Class|BPIHeistBasesLockdown|OnLockdownExpired(Message) _array_element (ScalableFloat|GetValueasBool (Variables|Hotfix|GetHF_LockdownExpirationDisablesVisuals)) (ScalableFloat|GetValueasBool (Variables|Hotfix|GetHF_LockdownExpirationDisablesSounds)))))))

(event OnComponentBeginOverlap(HeistBaseAreaCollision)
  (params (OverlappedComponent :type Primitive%20Component%20Object%20Reference) (OtherActor :type Actor%20Object%20Reference) (OtherComp :type Primitive%20Component%20Object%20Reference) (OtherBodyIndex :type Integer) (bFromSweep :type Boolean) (SweepResult :type Hit%20Result%20Structure%20%28by%20ref%29))
  ;[0.2226 0.2539 1] HeistBaseAreaCollision::
  ; OnComponentBeginOverlap()
  ; OnComponentEndOverlap()
  (CallFunction|AddOrRemovePlayer OtherActor true))

(event OnComponentEndOverlap(HeistBaseAreaCollision)
  (params (OverlappedComponent :type Primitive%20Component%20Object%20Reference) (OtherActor :type Actor%20Object%20Reference) (OtherComp :type Primitive%20Component%20Object%20Reference) (OtherBodyIndex :type Integer))
  ;[0.2226 0.2539 1] HeistBaseAreaCollision::
  ; OnComponentBeginOverlap()
  ; OnComponentEndOverlap()
  (CallFunction|AddOrRemovePlayer OtherActor))

(event EventEndPlay
  (params (EndPlayReason :type EEndPlayReason%20Enum :default "Destroyed"))
  (bind _lockdown_request_gameplaymessagetask (Variables|Default|GetLockdown%20Request%20GameplayMessageTask))
  ;[1 0.25 0] END PLAY
  ; Cleanup on endplay: timers, bind events, etc.
  (Utilities|FlowControl|Sequence
    (:then_0
      ;;[1 0.25 0] Sequence[0] General Cleanup
      (Protected|Utility|EndPlayCleanup)
      (Networking|SwitchHasAuthority
        (:Authority
          ;;[1 0 0.8325] LOG: Log the end event - debugging why Lockdown isn't consistently triggered
          (Fortnite|FortShippingLog self (Utilities|String|Append "HEIST BASE: Manager End Play Called. Reason = " (Utilities|String|EnumtoString EndPlayReason))))))
    (:then_1
      ;;[1 0.25 0] Sequence[1] Clear and Invalidate Timers
      (Utilities|Time|ClearandInvalidateTimerbyHandle (Variables|Lockdown|GetTH_LockdownExpirationTimer)))
    (:then_2
      ;;[1 0.25 0] Sequence[2] - Unbind Events
      (Utilities|IsValid _lockdown_request_gameplaymessagetask
        (:Is%20Valid
          (EventDispatchers|UnbindallEventsfromOnMessageReceived _lockdown_request_gameplaymessagetask))
        (:Is%20Not%20Valid
          _))
      (bind _cachedactorregistrydataobject (Variables|Default|GetCachedActorRegistryDataObject)
        (:then
          (EventDispatchers|UnbindEventfromOnActorRegistered _cachedactorregistrydataobject (EventDispatchers|CreateEvent :!event Cache%20Registered%20Base%20Defense)))
        (:else
          _))
      (bind _gamephaseupdatedasynctask (Variables|Default|GetGamePhaseUpdatedAsyncTask)
        (:then
          (GameplayEventRouter|StopListeningtoStatefulEvent _gamephaseupdatedasynctask))
        (:else)))
    (:then_3
      ;;[1 0.25 0] Sequence[3] - Call parent EndPlay last so inherited cleanup isn't run before other cleanup
      (|CallParentFunction EndPlayReason))))

(fn AddorRemovePlayer
  (params (PlayerActor :type Actor%20Object%20Reference) (AddPlayer :type Boolean))
  (bind _heistbasestatus_tag (Variables|Default|GetHeistBaseStatus%20Tag))
  (bind _heistbaseid_tag_const (Variables|Default|GetHeistBaseId%20Tag%20Const))
  (bind _array (Utilities|Array|MakeArray _heistbaseid_tag_const _heistbasestatus_tag))
  (bind _returnvalue (GameplayTags|MakeGameplayTagContainerfromArray _array))
  (bind _playerpawnsinheistbase (Variables|Default|GetPlayerPawnsInHeistBase))
  (bind _replicatepawntags_const (Variables|Default|GetReplicatePawnTags%20Const))
  (bind _asfort_player_pawn_athena (Utilities|Casting|CastToFortPlayerPawnAthena PlayerActor))
  (if AddPlayer
    (bind _returnvalue_2 (Utilities|Array|AddUnique _playerpawnsinheistbase _asfort_player_pawn_athena))
    (if (!= _returnvalue_2 -1)
      (Ability|GameplayEffect|AddLooseGameplayTags _asfort_player_pawn_athena _returnvalue _replicatepawntags_const) ; We successfully added this player to our list, give the player tags for this heist base and its current status.
      ;[0.2435 0.1798 0.4062] Analytics: Store players and the time they enter in order to get a total count and completion times for the base
      (if (not (Pawn|IsBotControlled _asfort_player_pawn_athena))
        (Utilities|Map|Add (Variables|Analytics|GetAnalyticsMapPlayerToEntryTime) _asfort_player_pawn_athena (Utilities|Time|GetGameTimeinSeconds))))
    (else
      (bind _returnvalue_3 (Utilities|Array|RemoveItem _playerpawnsinheistbase _asfort_player_pawn_athena))
      (if _returnvalue_3
        (Ability|GameplayEffect|RemoveLooseGameplayTags _asfort_player_pawn_athena _returnvalue _replicatepawntags_const) ; We successfully removed this player to our list, remove from this player the tags for this heist base and its current status.
        ;[0.2435 0.1798 0.4062] Analytics: Store when the lockdown-initiating player escapes
        (bind _lockdowntriggeringpawnathena (Variables|Analytics|GetLockdownTriggeringPawnAthena)
          (:then
            (if (== _asfort_player_pawn_athena _lockdowntriggeringpawnathena)
              (Variables|Analytics|SetLockdownPlayerEscapeTime (- (Utilities|Time|GetGameTimeinSeconds) (Utilities|Map|Find (Variables|Analytics|GetAnalyticsMapPlayerToEntryTime) _lockdowntriggeringpawnathena)))
              ;[0.2435 0.1798 0.4062] Analytics: Is base within storm when the lockdown-initiating player escapes the base?
              (bind _returnvalue_4 (SafeZone|IsLocationinSafeZone (Transformation|GetActorLocation)))
              (Variables|Analytics|SetAnalytic_BaseInStormOnEscape (not _returnvalue_4))))
          (:else))))))

(fn CacheRegisteredBaseDefense
  (params (GameplayTag :type Gameplay%20Tag%20Structure%20%28by%20ref%29) (RegisteredActor :type Actor%20Object%20Reference))
  ;[0.4973 0.715 1] CacheRegisteredBaseDefense
  ; Add the registered defense actor to this heist base - tag was already checked by the registry.
  (Utilities|Array|AddUnique (Variables|BaseDefense|GetCachedBaseDefenses) RegisteredActor))

(fn EditorRegisterBaseDefenses
  (bind _self self)
  (bind _returnvalue (Transformation|GetActorLocation _self))
  (bind _editorbasedefensedetectionradius (Variables|LevelDesign|GetEditorBaseDefenseDetectionRadius))
  ;[0 0 0] EditorRegisterBaseDefenses
  ; Get all objects within the specified radius, and register them with the current base if they are valid defenses (has the lockdown interface and registration component)
  (bind _outhits (Collision|MultiSphereTraceForObjects (+ _returnvalue -1.0) _returnvalue _editorbasedefensedetectionradius (Utilities|Array|MakeArray "ObjectTypeQuery1" "ObjectTypeQuery2" "ObjectTypeQuery6") true)) ; get all the objects in the base radius\nrequires complex trace for objects that don't have \nsimple collision (ex. Lockdown_Spot_C)
  (for _array_element _outhits
    (if (Utilities|DoesObjectImplementInterface (.HitActor (Collision|BreakHitResult _array_element)) "/HeistBasesGameplay/Items/BPI_HeistBases_Lockdown.BPI_HeistBases_Lockdown_C")
      (Utilities|IsValid (Actor|GetComponentbyClass (.HitActor (Collision|BreakHitResult _array_element)) "/Script/FortniteGame.FortAutoRegisterActorComponent")
        (:Is%20Valid
          (Transactions|TransactObject (.HitActor (Collision|BreakHitResult _array_element))) ; makes asset dirty to save tag update
          (Variables|/Script/FortniteGame.FortAutoRegisterActorComponent|SetActorRegistrationTags (GameplayTags|MakeGameplayTagContainerfromTag (Variables|Default|GetHeistBaseId%20Tag%20Const)) (Actor|GetComponentbyClass (.HitActor (Collision|BreakHitResult _array_element)) "/Script/FortniteGame.FortAutoRegisterActorComponent"))
          ;;[0.8203 0 1] Track count of base defenses found for logging
          (Math|Integer|IncrementInt (LocalVariables|Geth_CountBaseDefensesFound)))) ; get and confirm it has the AutoRegister component
    ))
  ;;[0.8886 0 1] Debug logging and visuals to confirm the registration has been successful.
  (Rendering|Debug|DrawDebugSphere _returnvalue _editorbasedefensedetectionradius 12 "(R=1.000000,G=0.000000,B=0.928250,A=1.000000)" 12.0 2e+01)
  (Fortnite|FortShippingLog self (Utilities|String|Append "HEIST BASE: Found and registered  " (Utilities|String|ToString(Integer) (LocalVariables|Geth_CountBaseDefensesFound)) " defenses and props for lockdown " (GameplayTags|GetDebugStringfromGameplayTag (Variables|Default|GetHeistBaseId%20Tag%20Const)))))

(fn EndPlayCleanup
  (Networking|SwitchHasAuthority
    (:Authority
      (Protected|AI|ReserveAIDensityReservoir true))
    (:Remote
      (Rendering|Material|SetScalarParameterValue "/Game/Packages/Fortress_SharedMaterials/GlobalMaterialParameters/FortniteMaterialParameters.FortniteMaterialParameters" "LockdownActive" :!class /Script/Engine.KismetMaterialLibrary) ; reset lockdown visuals for all emissive materials that get Lockdown effects\nif this is not reset, other bases will also appear to be in Lockdown
    )))

(fn GenerateAnalyticsLogMessage
  (params (AnalyticAttributeName :type String) (AnalyticsValue :type String))
  ;[0.2435 0.1798 0.4062] GenerateAnalyticsLogMessage
  ; Format the Log with a standard string for analytics
  (Fortnite|FortShippingLog self (Utilities|String|Append "HEIST BASE ANALYTICS: (" (Analytics|GetAnalyticsEventNameforBase) ") will send" AnalyticAttributeName " = " AnalyticsValue)))

(fn GetAnalyticsAttributes
  (bind _analytic_baseinstormonescape (Variables|Analytics|GetAnalytic_BaseInStormOnEscape))
  (bind _lockdownplayerescapetime (Variables|Analytics|GetLockdownPlayerEscapeTime))
  (bind _analytic_waslockeddown (Variables|Analytics|GetAnalytic_WasLockedDown))
  (bind _analytic_lockdowntime (Variables|Analytics|GetAnalytic_LockdownTime))
  (bind _analytic_baseinstormonlockdown (Variables|Analytics|GetAnalytic_BaseInStormOnLockdown))
  (bind _analytic_playerlockdowntime (Variables|Analytics|GetAnalytic_PlayerLockdownTime))
  ;[0.2435 0.1798 0.4062] GetAnalyticsAttributes
  ; Get and log all tracked attributes for the Heist Base
  (Utilities|Array|Clear (LocalVariables|GetReturn_LocalAnalytics))
  (Utilities|FlowControl|Sequence
    (:then_0
      ;;[0.2435 0.1798 0.4062] Heist Base Analytics: attributes related to the base iteself and its status
      (bind _returnvalue (Analytics|ConstructAnalyticsAttribute :AttributeName "LockedDownDuringMatch" :IntValue -1 :floatValue -1.0 :BoolValue _analytic_waslockeddown)) ; LockedDownDuringMatch\nwhether the base was locked down or not during the match
      (Utilities|Array|Add (LocalVariables|GetReturn_LocalAnalytics) _returnvalue)
      (Analytics|GenerateAnalyticsLogMessage "LockedDownDuringMatch" (Utilities|String|ToString(Boolean) _analytic_waslockeddown))
      (bind _returnvalue_2 (Analytics|ConstructAnalyticsAttribute :AttributeName "LockdownInitiatedMatchTime" :IntValue -1 :floatValue _analytic_lockdowntime :BoolValue false)) ; LockdownInitiatedMatchTime\ntime when the base enters lockdown
      (Utilities|Array|Add (LocalVariables|GetReturn_LocalAnalytics) _returnvalue_2)
      (Analytics|GenerateAnalyticsLogMessage "LockdownInitiatedMatchTime" (Utilities|String|ToString(Float) _analytic_lockdowntime))
      (bind _returnvalue_3 (Analytics|ConstructAnalyticsAttribute :AttributeName "StormPresentOnLockdown" :IntValue -1 :floatValue -1.0 :BoolValue _analytic_baseinstormonlockdown)) ; StormPresentOnLockdown\nif the base manager was within the storm when lockdown was initiated
      (Utilities|Array|Add (LocalVariables|GetReturn_LocalAnalytics) _returnvalue_3)
      (Analytics|GenerateAnalyticsLogMessage "StormPresentOnLockdown" (Utilities|String|ToString(Boolean) _analytic_baseinstormonlockdown))
      (bind _returnvalue_4 (Analytics|ConstructAnalyticsAttribute :AttributeName "StormPresentOnEscape" :IntValue -1 :floatValue -1.0 :BoolValue _analytic_baseinstormonescape)) ; StormPresentOnEscape\nIf the base manager was within the storm when the lockdown initiating player escapes the base
      (Utilities|Array|Add (LocalVariables|GetReturn_LocalAnalytics) _returnvalue_4)
      (Analytics|GenerateAnalyticsLogMessage "StormPresentOnEscape" (Utilities|String|ToString(Boolean) _analytic_baseinstormonescape)))
    (:then_1
      ;;[0.2435 0.1798 0.4062] Heist Player Analytics: attributes related to the players in the base
      (bind _returnvalue_5 (Analytics|ConstructAnalyticsAttribute :AttributeName "LockdownInitiatedPlayerTime" :IntValue -1 :floatValue _analytic_playerlockdowntime :BoolValue false)) ; LockdownInitiatedPlayerTime\nTime from the player who initiates lockdown from when they\n enter the base to starting lockdown (claiming item)
      (Utilities|Array|Add (LocalVariables|GetReturn_LocalAnalytics) _returnvalue_5)
      (Analytics|GenerateAnalyticsLogMessage "LockdownInitiatedPlayerTime" (Utilities|String|ToString(Float) _analytic_playerlockdowntime))
      (bind _returnvalue_6 (Analytics|ConstructAnalyticsAttribute :AttributeName "LockdownEscapePlayerTime" :IntValue -1 :floatValue _lockdownplayerescapetime :BoolValue false)) ; LockdownEscapePlayerTime\nTime for the player who initiates lockdown from when they\n enter the base to when they escape after claiming the reward
      (Utilities|Array|Add (LocalVariables|GetReturn_LocalAnalytics) _returnvalue_6)
      (Analytics|GenerateAnalyticsLogMessage "LockdownEscapePlayerTime" (Utilities|String|ToString(Float) _lockdownplayerescapetime)))
    (:then_2
      ;;[0.9707 0.4876 1] RETURN: Send all the tracked analytic events
      (return (LocalVariables|GetReturn_LocalAnalytics)))))

(fn GetAnalyticsEventNameForBase
  (bind (_value _returnvalue) (Utilities|Map|Find (Variables|Analytics|GetAnalyticsMapBaseInstanceEventName) (Variables|Default|GetHeistBaseId%20Tag%20Const)))
  (return (Utilities|String|SelectString _value "Core.HeistBase.ERRUndefined" _returnvalue)))

(fn GetAndRegisterDefenses
  (bind _returnvalue (WorldSubsystems|GetFortGlobalActorRegistrySubsystem))
  (bind _heistbaseid_tag_const (Variables|Default|GetHeistBaseId%20Tag%20Const))
  ;[0.4973 0.715 1] GetAndRegisterDefenses
  ; Find all the base defesnes for this managed base and store them for future use
  (Networking|SwitchHasAuthority
    (:Authority
      (bind _outactors (Class|FortGlobalActorRegistrySubsystem|GetRegisteredActors _returnvalue _heistbaseid_tag_const)) ; get all the current registered actors
      ;;[0.4973 0.715 1] Add all found actors - run additional validation if necessary
      (for _array_element _outactors
        (CallFunction|CacheRegisteredBaseDefense _heistbaseid_tag_const _array_element))
      ;;[0.4973 0.715 1] Bind to Data Registry object - Will register any defenses that are added to the registry object after the initial run
      (bind _returnvalue_2 (Class|FortGlobalActorRegistrySubsystem|GetActorRegistryDataforGameplayTag _returnvalue _heistbaseid_tag_const))
      (Utilities|IsValid _returnvalue_2
        (:Is%20Valid
          (bind _output_get (Variables|Default|SetCachedActorRegistryDataObject _returnvalue_2))
          (EventDispatchers|BindEventtoOnActorRegistered _output_get (EventDispatchers|CreateEvent :!event Cache%20Registered%20Base%20Defense)))))))

(fn GetMaxAIToAlert
  (bind _hotfix_lockdownaialertmaxtoalert (Variables|Hotfix|GetHotfix_LockdownAIAlertMaxToAlert))
  (bind _currentplaylistmaxteamsize (Variables|AI|GetCurrentPlaylistMaxTeamSize))
  (if (ScalableFloat|GetValueasBool (Variables|Hotfix|GetHotfix_AIScalabilityTuningEnabled))
    (if (Utilities|Array|IsValidIndex _hotfix_lockdownaialertmaxtoalert _currentplaylistmaxteamsize)
      (return (ScalableFloat|GetValueasInteger (Utilities|Array|Get(acopy) _hotfix_lockdownaialertmaxtoalert _currentplaylistmaxteamsize)))
      (else
        _))
    (else
      _))
  (return (ScalableFloat|GetValueasInteger (Variables|Hotfix|GetHotfix_DefaultMaxAIToAlert))))

(fn HandleGamePhaseChanged
  (params (GamePhase :type EAthenaGamePhase%20Enum))
  ;[0.03332 0.3212 0.2859] HandleGamePhaseChanged
  ; When the game is ended, send the assigned objective analytics event for the match
  (if (Utilities|Enum|Equal(Enum) GamePhase "EndGame")
    ;;[0.2435 0.1798 0.4062] Get and Fire all Analytic Attributes for the match (each objective will do this)
    (bind _analyticsattributes (Analytics|GetAnalyticsAttributes))
    (Analytics|FireAnalyticsEventBlueprintArray (Analytics|GetAnalyticsEventNameforBase) _analyticsattributes)
    (bind _gamephaseupdatedasynctask (Variables|Default|GetGamePhaseUpdatedAsyncTask)
      (:then
        (GameplayEventRouter|StopListeningtoStatefulEvent _gamephaseupdatedasynctask))
      (:else))))

(fn HeistBaseEndLockdown
  (params (Location%20Id%20Substring :type String))
  (Networking|SwitchHasAuthority
    (:Authority
      (if (and (Utilities|String|Contains (GameplayTags|GetDebugStringfromGameplayTag (Variables|Default|GetHeistBaseId%20Tag%20Const)) Location%20Id%20Substring false true) (Variables|Lockdown|GetbInLockdown))
        (CallFunction|EndLockdownbyExpiration)))))

(fn HeistBaseTriggerLockdown
  (params (Location%20Id%20Substring :type String))
  (Networking|SwitchHasAuthority
    (:Authority
      (if (and (Utilities|String|Contains (GameplayTags|GetDebugStringfromGameplayTag (Variables|Default|GetHeistBaseId%20Tag%20Const)) Location%20Id%20Substring false true) (not (Variables|Lockdown|GetbInLockdown)))
        (CallFunction|OnRequestLockdownBase (Utilities|Struct|MakeStructHeistBasesLockdownInitiatedPayload (Variables|Default|GetHeistBaseId%20Tag%20Const) 0 self))))))

(fn OnRep_bInLockdown
  (bind _binlockdown (Variables|Lockdown|GetbInLockdown))
  (Fortnite|FortShippingLog self (Utilities|String|Append "OnRepBInLockDown() [" (Utilities|GetDisplayName self) "] bInLockdown: [" (Utilities|String|ToString(Boolean) _binlockdown) "]."))
  (Networking|CanExecuteCosmeticEvents
    (:True
      (if _binlockdown
        (Rendering|Material|SetScalarParameterValue "/Game/Packages/Fortress_SharedMaterials/GlobalMaterialParameters/FortniteMaterialParameters.FortniteMaterialParameters" "LockdownActive" (Math|Conversions|ToFloat(Boolean) _binlockdown) :!class /Script/Engine.KismetMaterialLibrary) ; turn on lockdown visuals when locked down
        (Default|CallLockdown (Variables|Default|GetHeistBaseId%20Tag%20Const))))
    (:False
      (if (Variables|Lockdown|GetbInLockdown)
        (Variables|Analytics|SetAnalytic_WasLockedDown true)))))

(fn OnRep_bLockdownExpired
  (bind _hf_lockdownexpirationdisablesvisuals (Variables|Hotfix|GetHF_LockdownExpirationDisablesVisuals))
  (bind _returnvalue (ScalableFloat|GetValueasBool _hf_lockdownexpirationdisablesvisuals))
  (Networking|CanExecuteCosmeticEvents
    (:True
      (if (Variables|Lockdown|GetbLockdownExpired)
        (Rendering|Material|SetScalarParameterValue "/Game/Packages/Fortress_SharedMaterials/GlobalMaterialParameters/FortniteMaterialParameters.FortniteMaterialParameters" "LockdownActive" (Math|Conversions|ToFloat(Boolean) (not _returnvalue)) :!class /Script/Engine.KismetMaterialLibrary) ; set lockdown emissive visuals when expired
        (Default|CallLockdownExpiredClient (Variables|Default|GetHeistBaseId%20Tag%20Const) _returnvalue (ScalableFloat|GetValueasBool (Variables|Hotfix|GetHF_LockdownExpirationDisablesSounds)))))))

(fn ReserveAIDensityReservoir
  (params (bUnreserve :type Boolean))
  (bind _self self)
  (Networking|SwitchHasAuthority
    (:Authority
      (bind _livingworldmanager (Utility|GetValidLivingWorldManager)
        (:Is%20Valid
          (if (not bUnreserve)
            (Class|FortAthenaLivingWorldManager|RegisterDensityReservoir _livingworldmanager _self (Variables|AI|Budget|GetReservedReservoirDensity) (Variables|AI|Budget|GetReservedReservoirRange))
            (Fortnite|FortShippingLog _self "Registered density reservoir")
            (else
              (Class|FortAthenaLivingWorldManager|UnregisterDensityReservoir _livingworldmanager _self)
              (Fortnite|FortShippingLog _self "Unregistered self from density reservoir"))))))))

(fn SendLockdownStimToGuards
  (Networking|SwitchHasAuthority
    (:Authority
      (LocalVariables|SetLMaxAIToAlert (CallFunction|GetMaxAIToAlert))
      ;[0.1708 1 0.2319] Set Alert location to the world location of the base manager and the offset
      (bind _output_get (LocalVariables|SetLLockdownAlertLocation (+ (Variables|Lockdown|GetLockdownAlertLocationOffset) (Transformation|GetActorLocation))))
      ;[0.1407 0.1494 1] Trace for pawn capsules in an area around the lockdown locaiton
      (bind _outhits (Collision|MultiSphereTraceByChannel :Start _output_get :End (+ _output_get "0.000000,0.000000,1.000000") :Radius (ScalableFloat|GetValueatLevel (Variables|Hotfix|GetHotfix_LockdownAIAlertRange)) :TraceChannel "TraceTypeQuery9" :bTraceComplex false :DrawDebugType "ForDuration" :bIgnoreSelf true :TraceColor "(R=1.000000,G=0.000000,B=0.000000,A=1.000000)" :TraceHitColor "(R=0.000000,G=1.000000,B=0.000000,A=1.000000)" :DrawTime 5.0))
      ;[0.1407 0.1494 1] Send alert stimulus to base ai, limiting by count and the closest ai (hitresult actors are sorted by distance along the trace. I tested and verified this)
      (bind _array_element (Utilities|Array|ForEachLoopwithBreak _outhits)
        (:LoopBody
          (if (< (LocalVariables|GetLAlertedGuardsCount) (LocalVariables|GetLMaxAIToAlert))
            _
            (else
              _))
          (bind _asfort_player_pawn (Utilities|Casting|CastToFortPlayerPawn (.HitActor (Collision|BreakHitResult _array_element))))
          (if (GameplayTags|HasTag (Fortnite|Factions|GetFactionTagContainerforActor _asfort_player_pawn) (Variables|Default|GetHeistFactionTag) true)
            (bind _asfort_athena_aibot_controller (Utilities|Casting|CastToFortAthenaAIBotController (Pawn|GetController _asfort_player_pawn)))
            (Class|FortAthenaAIBotController|QueueStim _asfort_athena_aibot_controller self (LocalVariables|GetLLockdownAlertLocation) (Math|Vector|GetUnitDirection(Vector) (Transformation|GetActorLocation _asfort_player_pawn) (LocalVariables|GetLLockdownAlertLocation)) "Enemy" 1.0 "None")
            (Math|Integer|IncrementInt (LocalVariables|GetLAlertedGuardsCount))))
        (:Completed
          _))
      ;[1 0.3042 0.1474] Log alert results
      (Fortnite|FortShippingLog self (Utilities|String|Append (Utilities|String|ToString(Name) (GameplayTags|GetTagName (Variables|Default|GetHeistBaseId%20Tag%20Const))) "Lockdown alerted guards count = " (Utilities|String|ToString(Integer) (LocalVariables|GetLAlertedGuardsCount)))))))

(fn SetStatus
  (params (New%20Status%20Tag :type Gameplay%20Tag%20Structure))
  (bind _replicatepawntags_const (Variables|Default|GetReplicatePawnTags%20Const))
  (if (!= (Variables|Default|GetHeistBaseStatus%20Tag) New%20Status%20Tag)
    (Utilities|FlowControl|Sequence
      (:then_0
        ;[0.1874 0.9901 1] Update the status tags of all players within this heist base.
        (for _array_element (Variables|Default|GetPlayerPawnsInHeistBase)
          (Utilities|IsValid _array_element
            (:Is%20Valid
              (Ability|GameplayEffect|RemoveLooseGameplayTags _array_element (GameplayTags|MakeGameplayTagContainerfromTag (Variables|Default|GetHeistBaseStatus%20Tag)) _replicatepawntags_const)
              (Ability|GameplayEffect|AddLooseGameplayTags _array_element (GameplayTags|MakeGameplayTagContainerfromTag (Variables|GetNew%20Status%20Tag)) _replicatepawntags_const)))))
      (:then_1
        (Variables|Default|SetHeistBaseStatus%20Tag New%20Status%20Tag)))))

(fn UnRegisterDefenses
  (Utilities|Array|Clear (Variables|BaseDefense|GetCachedBaseDefenses)))

(fn ConstructionScript)
