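; Blueprint interface for Heist Bases items, such as base defenses, that respond to a base lockdown.
; Declares a lockdown-initiated event where each defense runs its unique lockdown behavior and returns a boolean.
; An expiry event takes flags to disable visuals and sounds, and a separate event handles lockdown reset.
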(fn OnLockdownExpired
  (params (DisableVisuals :type Boolean) (DisableSounds :type Boolean)))

(fn OnLockdownInitiated
  ;[1 1 1] Interface function for base defenses to initiate their unique lockdown behavior
  (return false))

(fn OnLockdownReset)
