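; Server-authoritative heist base component that activates one World Partition defense data layer at a time from a supplied list.
; InitializeDataLayers stores the list and loads a random entry, or the first under base defense determinism.
; Loading unregisters defenses through the owning BGA_HeistBaseManager, unloads the current layer, activates the new one, then re-registers defenses.
; Cheat events set a layer by index or cycle with wraparound, optionally on a looping debug timer.
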(event EventBeginPlay
  ;[0 1 0] BeginPlay - Cache reference to owning Manager
  (Utilities|SwitchHasAuthority
    (:Authority
      (bind _asbga_heist_base_manager (Utilities|Casting|CastToBGA_HeistBaseManager (Components|GetOwner)))
      (Variables|Default|SetHeistBaseManager _asbga_heist_base_manager))))

(event EventEndPlay
  (params (EndPlayReason :type EEndPlayReason%20Enum :default "Destroyed"))
  ;[1 0.1875 0.1875] EndPlay
  (|CallParentFunction EndPlayReason)
  (Utilities|Time|ClearandInvalidateTimerbyHandle (Variables|Debug|GetDebugCycleTimer)))

(event Custom|LoadRandomDataLayer
  (bind _defensedatalayers (Variables|Default|GetDefenseDataLayers))
  ;[0.2 0.2 1] LoadRandomDataLayer
  (Utilities|SwitchHasAuthority
    (:Authority
      (Variables|Default|GetCurrentDataLayer
        (:then)
        (:else
          (if (> (Utilities|Array|Length _defensedatalayers) 0)
            (Private|LoadDataLayer (select (Class|FortDeterminismConfig|IsUsingBaseDefenseDeterminism) (Utilities|Array|Get(acopy) _defensedatalayers) (Utilities|Array|RandomArrayItem _defensedatalayers))) ; Object validated internally
          ))) ; We should never load more than \n1 DataLayer in the normal Flow of Events
    )))

(event Custom|SetDefenseDataLayer
  (params (Index :type Integer))
  (bind _defensedatalayers (Variables|Default|GetDefenseDataLayers))
  (bind _output (Utilities|Array|Get(acopy) _defensedatalayers Index))
  ;[1 1 0] SetDefenseDatalayer
  ; cheat ke* SetDefenseDataLayer (Index)
  (Networking|SwitchHasAuthority
    (:Authority
      (if (Utilities|Array|IsValidIndex _defensedatalayers Index)
        (Utilities|IsValid _output
          (:Is%20Valid
            (Private|LoadDataLayer _output)))))))

(event Custom|CycleDefenseDataLayer
  ;[1 1 0] CycleDefenseDataLayer
  ; cheat ke* CycleDefenseDataLayer
  (Networking|SwitchHasAuthority
    (:Authority
      (bind _datalayer (Private|GetNextDataLayer))
      (Utilities|IsValid _datalayer
        (:Is%20Valid
          (Private|LoadDataLayer _datalayer))))))

(fn GetNextDataLayer
  (bind _defensedatalayers (Variables|Default|GetDefenseDataLayers))
  (bind _defensedatalayers_2 (Variables|Default|GetDefenseDataLayers))
  (Utilities|FlowControl|Sequence
    (:then_0
      ;[0 0 0] Try and find the CurrentDataLayer in our array
      (bind _currentdatalayer (Variables|Default|GetCurrentDataLayer)
        (:then
          (LocalVariables|SetLCurrentDataLayerIndex (Utilities|Array|FindItem (Variables|Default|GetDefenseDataLayers) _currentdatalayer)))
        (:else)))
    (:then_1
      ;[0 0 0] Increment the LCurrentDataLayerIndex and early return if the data layer is valid
      (bind _result (Math|Integer|IncrementInt (LocalVariables|GetLCurrentDataLayerIndex))
        (:%20%20
          (if (Utilities|Array|IsValidIndex _defensedatalayers _result)
            (return (Utilities|Array|Get(acopy) _defensedatalayers _result))))))
    (:then_2
      ;[0 0 0] Check the Array element at 0 and early return if it is valid
      (if (Utilities|Array|IsValidIndex _defensedatalayers_2)
        (return (Utilities|Array|Get(acopy) _defensedatalayers_2))))
    (:then_3
      ;[1 0 0] Failed to find data layer
      (Fortnite|FortShippingLog self "GetNextDataLayer failed to find valid data layer asset" true)
      (return 0))))

(fn InitializeDataLayers
  (params (RelevantDataLayers :type Array%20of%20Data%20Layer%20Asset%20Object%20References))
  (Utilities|SwitchHasAuthority
    (:Authority
      (bind _output_get (Variables|Default|SetDefenseDataLayers RelevantDataLayers))
      (bind _output_get_2 (Variables|Default|SetTotalDataLayers (Utilities|Array|Length _output_get)))
      (if (> _output_get_2 0)
        (CallFunction|LoadRandomDataLayer)
        (if (Variables|Debug|GetbDebugCycleDataLayers)
          (bind _returnvalue (Utilities|Time|SetTimerbyEvent (EventDispatchers|CreateEvent :!event CycleDefenseDataLayer) (Variables|Debug|GetDebugCycleTime) true))
          (Variables|Debug|SetDebugCycleTimer _returnvalue))))))

(fn LoadDataLayer
  (params (DataLayerToLoad :type Data%20Layer%20Asset%20Object%20Reference))
  (Networking|SwitchHasAuthority
    (:Authority
      (CallFunction|PrintDebugMessage (Utilities|String|Append "HeistBase Datalayer - Try Load " (Utilities|GetDisplayName DataLayerToLoad)))
      (bind _returnvalue (WorldPartition|GetDataLayerManager))
      (if (and (Utilities|IsValid _returnvalue :!class /Script/Engine.KismetSystemLibrary) (Utilities|IsValid DataLayerToLoad :!class /Script/Engine.KismetSystemLibrary))
        (Private|UnloadCurrentDatalayer)
        (bind _returnvalue_2 (DataLayers|SetDataLayerRuntimeState _returnvalue DataLayerToLoad "Activated" false :!class /Script/Engine.DataLayerManager))
        (if _returnvalue_2
          (bind _output_get (Variables|Default|SetCurrentDataLayer DataLayerToLoad))
          (CallFunction|PrintDebugMessage (Utilities|String|Append "HeistBase Datalayer - Successfully loaded " (Utilities|GetDisplayName _output_get)))
          (Utilities|FlowControl|Sequence
            (:then_0
              (bind _heistbasemanager (Variables|Default|GetHeistBaseManager)
                (:then
                  (Class|BGAHeistBaseManager|GetandRegisterDefenses _heistbasemanager))
                (:else)))
            (:then_1
              (return true)))
          (else
            _))
        (else
          ;[1 0 0] Log Invalid objects and return unsuccessful
          (Fortnite|FortShippingLog self "LoadDataLayer:: New Data Layer / DataLayerManager not valid" true)))
      ;[1 0 0] Log failure to set runtime state and return unsuccessful
      (Fortnite|FortShippingLog self "LoadDataLayer:: Set Data Layer Runtime state failed" true)
      (return false))
    (:Remote
      (return false))))

(fn PrintDebugMessage
  (params (Message :type String) (bPrintToScreen :type Boolean :default true) (Duration :type Float%20%28single-precision%29 :default 3.0))
  (if (Variables|Default|GetPrintDebug)
    (Development|PrintString Message bPrintToScreen true "(R=0.000000,G=0.660000,B=1.000000,A=1.000000)" Duration)))

(fn UnloadCurrentDatalayer
  (Utilities|SwitchHasAuthority
    (:Authority
      (bind _returnvalue (WorldPartition|GetDataLayerManager))
      (Utilities|IsValid _returnvalue
        (:Is%20Valid
          (bind _currentdatalayer (Variables|Default|GetCurrentDataLayer)
            (:then
              (bind _heistbasemanager (Variables|Default|GetHeistBaseManager)
                (:then
                  (Class|BGAHeistBaseManager|UnRegisterDefenses _heistbasemanager)
                  (bind _returnvalue_2 (DataLayers|SetDataLayerRuntimeState _returnvalue _currentdatalayer "Unloaded" false :!class /Script/Engine.DataLayerManager))
                  (return _returnvalue_2))
                (:else
                  _)))
            (:else
              _))
          (return false))
        (:Is%20Not%20Valid
          _)))))
